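# Looping sound cue for the Agent glider cosmetic's flight audio.
# A looping wave player feeds the flight loop wave into an Enveloper, whose volume curve ramps up from a low level over the first fraction of a second.
# The cue plays at reduced volume with player glider thrust attenuation, the cosmetic glider loop audio properties sheet and a glider sound class.
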SoundCue /BRCosmetics/Sounds/Gliders/Agent/Glider_Agent_Loop_Cue
root=Enveloper_0
nodes [
 {
  id=Enveloper_0
  children [WavePlayer_0]
  properties {pitchCurve {editorCurveData {keys [{time=0}]}}, volumeCurve {editorCurveData {keys [<at 0 0.2>,<at 0.3 1>]}}}
 }
 {id=WavePlayer_0, properties {bLooping=true, soundWaveAssetPtr=&/Glider_Agent_Flight_LP}}
]
properties {
 attenuationSettings=&Fort_Player_Glider_Thrust_Athena_Att
 audioPropertiesSheet=&APS_Cosmetic_Glider_Loop
 maxDistance=12500
 soundClassObject=&Fort_Gliders_Land
 volumeMultiplier=0.1
}
