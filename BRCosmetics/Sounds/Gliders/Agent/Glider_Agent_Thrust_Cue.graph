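# Sound cue for the thrust sound of the Agent glider cosmetic.
# A Random node picks one of three equally weighted thrust waves, and a Modulator at the root varies its playback.
# Glider-specific attenuation, concurrency, sound class and a cosmetic glider loop properties sheet govern how it plays, at a reduced volume.
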SoundCue /BRCosmetics/Sounds/Gliders/Agent/Glider_Agent_Thrust_Cue
root=Modulator_0
nodes [
 {id=Modulator_0, children [Random_1]}
 {id=Random_1, children [WavePlayer_0,WavePlayer_4,WavePlayer_5], properties {weights [1,1,1]}}
 {id=WavePlayer_0, properties {soundWaveAssetPtr=&/Glider_Agent_Thrust_01}}
 {id=WavePlayer_4, properties {soundWaveAssetPtr=&/Glider_Agent_Thrust_02}}
 {id=WavePlayer_5, properties {soundWaveAssetPtr=&/Glider_Agent_Thrust_03}}
]
properties {
 attenuationSettings=&Fort_Player_Glider_Thrust_Athena_Att
 audioPropertiesSheet=&APS_Cosmetic_Glider_Loop
 maxDistance=12500
 soundClassObject=&Fort_Gliders_Land
 volumeMultiplier=0.4
 concurrencySet [&Athena_Glider_Thrust_Conc]
}
